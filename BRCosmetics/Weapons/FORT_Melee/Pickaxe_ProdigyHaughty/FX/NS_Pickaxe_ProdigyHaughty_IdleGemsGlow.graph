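# Niagara system for the ProdigyHaughty pickaxe's idle gem glow, lighting gems and mesh pieces in steps based on User.HitCount.
# GlowGems pins ten sprites to skeletal mesh bones or sockets by partition index, in three size groups.
# Glowmeshes spawns two mesh particles, using partition index as visibility tag, at higher steps than the gems.
# Both compare their step to HitCount plus a system threshold bias, then smooth the on/off result into alpha for gradual fades.
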NiagaraSystem /BRCosmetics/Weapons/FORT_Melee/Pickaxe_ProdigyHaughty/FX/NS_Pickaxe_ProdigyHaughty_IdleGemsGlow
dependencies {
 usedDataInterfaces [&NiagaraDataInterfaceArrayFloat, &NiagaraDataInterfaceSkeletalMesh]
 usedDynamicInputs []
 usedModules [
  &self:SystemSpawnScript.SystemScriptSource.SystemScriptGraph.NiagaraNodeAssignment_1.SetVariables_#1
  &self:SystemSpawnScript.SystemScriptSource.SystemScriptGraph.NiagaraNodeAssignment_2.SetVariables_#2
  &SystemState
 ]
}
emitters {
 GlowGems {
  emitterUpdateScript [
   &EmitterState(Life Cycle Mode=::NewEnumerator1, Loop Behavior=::NewEnumerator1, Loop Duration Mode=::NewEnumerator1, UseLoopDelay=false)
   &SpawnBurst_Instantaneous(Spawn Count=10, Use Spawn Probability=false, UseLoopCountLimit=false)
  ]
  particleSpawnScript [
   &PartitionParticles(Number Of Partitions=10, Particles Per Partition=1, Partition Mode:ENiagara_PartitionMode=::NewEnumerator1)
   /SetVariables(Particles.GemGroup:NiagaraInt32=&Add_Integer(A=/Niagara/DynamicInputs/Divide/Divide_Int(A:NiagaraInt32=$Particles.PartitionParticles.Partition, B:NiagaraInt32=4), B=1)) {isSetParameters=true}
   &InitializeParticle(
    Color=<rgba 0.6 0.03 0.0395 1>
    Lifetime=2
    Sprite Rotation Mode=::NewEnumerator2
    Sprite Size Mode=::NewEnumerator3
    Uniform Sprite Size=&SelectFloatFromArray(Direct Array Index=$Particles.GemGroup, Float Selection Array=DataInterface() {propertyValues {FloatData [0,25,20,35]}})
    UsePositionOffset=false
    Write Material Random=true
   )
   /SetVariables(Particles.SpriteGlowSize:Vector2f=&Multiply_Vector2D_ByFloat(Float=0.4, Vector2D=$Particles.SpriteSize)) {isSetParameters=true}
  ]
  particleUpdateScript [
   &SkeletalMeshLocation(
    Bone / Socket Index=$Particles.PartitionParticles.Partition
    MaskSpawnGroup=false
    Mesh Sampling Type=::NewEnumerator5
    Skeletal Mesh=$System.Skeletal Mesh
    Socket Sampling Mode=::NewEnumerator5
   )
   &ParticleState(
    Kill Particles When Lifetime Has Elapsed=false
    Let Infinitely Lived Particles Die When Emitter Deactivates=true
    Loop Particles Lifetime=true
   )
   &SpriteRotationRate(Rotation Rate=20)
   &InterpolateOverTime(
    Rate Of Change=5
    Target Float Value:NiagaraFloat=&MakeCustomFloatFromBool(
     Bool=&SetBoolByIntComparison(
      A=$Particles.GemGroup
      B=&MakeIntFromFloat(Float=&Add_Float(A=$User.HitCount, B=$System.GlowThresholdBias))
      Comparison Type=::NewEnumerator7
     )
     True Float=1
     False Float=0
    )
   )
   &ScaleColor(
    Scale Alpha=&Multiply_Float(A=$Particles.InterpolateOverTime.InterpolateOverTime_Float.Moving Average, B=0.5)
    ScaleA=true
    ScaleRGB=true
   )
   /SetVariables(Particles.GlowIntenseColor:LinearColor=&MakeLinearColorFromVectorAndFloat(
     "Vector (RGB)"=&MakeVectorFromLinearColorRGB(Color=<rgba 3 0.42 0.85 1>)
     "Float (Alpha)"=&Multiply_Float(A=$Particles.InterpolateOverTime.InterpolateOverTime_Float.Moving Average, B=1)
    )) {isSetParameters=true}
  ]
  rendererClasses [&NiagaraSpriteRendererProperties]
  renderers [
   {rendererClass=&NiagaraSpriteRendererProperties}
   {rendererClass=&NiagaraSpriteRendererProperties, rendererIndex=1}
  ]
 }
 Glowmeshes {
  emitterUpdateScript [
   &EmitterState(Life Cycle Mode=::NewEnumerator1, Loop Behavior=::NewEnumerator1, Loop Duration Mode=::NewEnumerator1, UseLoopDelay=false)
   &SpawnBurst_Instantaneous(Spawn Count=2, Use Spawn Probability=false, UseLoopCountLimit=false)
  ]
  particleSpawnScript [
   &InitializeParticle(Color=<rgba 1 0.01 0.0265 1>, Mesh Scale Mode=::NewEnumerator3, UsePositionOffset=false, Write Material Random=true)
   &PartitionParticles(Particles Per Partition=1, Partition Mode:ENiagara_PartitionMode=::NewEnumerator1)
   /SetVariables(Particles.VisibilityTag=$Particles.PartitionParticles.Partition) {isSetParameters=true}
  ]
  particleUpdateScript [
   &ParticleState(Kill Particles When Lifetime Has Elapsed=false, Let Infinitely Lived Particles Die When Emitter Deactivates=true)
   &InterpolateOverTime(
    Rate Of Change=10
    Target Float Value:NiagaraFloat=&MakeCustomFloatFromBool(
     Bool=&SetBoolByIntComparison(
      A=&Add_Integer(A=$Particles.PartitionParticles.Partition, B=4)
      B=&MakeIntFromFloat(Float=&Add_Float(A=$User.HitCount, B=$System.GlowThresholdBias))
      Comparison Type=::NewEnumerator7
     )
     True Float=1
     False Float=0
    )
   )
   &ScaleColor(
    Scale Alpha=&Multiply_Float(A=$Particles.InterpolateOverTime.InterpolateOverTime_Float.Moving Average, B=1)
    ScaleA=true
    ScaleRGB=true
   )
  ]
  rendererClasses [&NiagaraMeshRendererProperties]
  renderers [
   {rendererClass=&NiagaraMeshRendererProperties, rendererIndex=0}
   {rendererClass=&NiagaraMeshRendererProperties, rendererIndex=1}
  ]
 }
}
summary {
 systemName=NS_Pickaxe_ProdigyHaughty_IdleGemsGlow
 emitters [
  {emitterName=GlowGems, rendererClasses [&NiagaraSpriteRendererProperties]}
  {emitterName=Glowmeshes, rendererClasses [&NiagaraMeshRendererProperties]}
 ]
 userVariables [
  User.HitCount {type=NiagaraFloat, defaultValue=0, description=""}
  User.Variant {type=NiagaraFloat, defaultValue=0, description=""}
 ]
}
