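# Niagara system that draws ribbon swing trails for the ProdigyHaughty pickaxe.
# Three anim-trail emitters follow skeletal mesh bone or socket pairs: a main blade trail, a near-black trail and a pair of small blade trails.
# User.HitCount indexes color arrays that switch from gold to red, with the small trails turning one index before the main trail.
# Alpha scales by time of day and particle life, then cuts off early by system age.
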NiagaraSystem /BRCosmetics/Weapons/FORT_Melee/Pickaxe_ProdigyHaughty/FX/NS_Pickaxe_ProdigyHaughty_Trails
dependencies {
 usedDataInterfaces [&NiagaraDataInterfaceArrayColor, &NiagaraDataInterfaceColorCurve, &NiagaraDataInterfaceCurve]
 usedDynamicInputs []
 usedModules [
  &self:SystemSpawnScript.SystemScriptSource.SystemScriptGraph.NiagaraNodeAssignment_0.SetVariables_#1
  &SystemState
 ]
}
emitters {
 BladeSmall {
  emitterUpdateScript [
   &EmitterState()
   &FN_SpawnAnimTrail(Number of Anim Trails=2, Spawn Interval=0)
  ]
  particleSpawnScript [
   &FN_InitializeAnimTrailBoneIndecies(First Bone or Socket Index=3, Second Bone or Socket Index=4)
   &FN_InitializeAnimTrailBoneIndecies(AnimTrail Selection=1, First Bone or Socket Index=5, Second Bone or Socket Index=6) {module=FN_InitializeAnimTrailBoneIndecies001}
   &AddVelocity(Rotation Coordinate Space=::World, Velocity=<0 0 40>) {module=AddVelocity001}
   &FN_InitializeAnimTrail(
    Color=&SelectLinearColorFromArray(
     Array Sampling Mode=::NewEnumerator1
     Direct Array Index=&MakeIntFromFloat(Float=$User.HitCount)
     Color Selection Array=DataInterface() {propertyValues {ColorData [
        <rgba 0.3 0.1893 0.054 1>
        <rgba 0.3 0.1893 0.054 1>
        <rgba 0.3 0.1893 0.054 1>
        <rgba 0.3 0.1893 0.054 1>
        <rgba 0.9 0.018 0.0327 1>
        <rgba 0.9 0.018 0.0327 1>
       ]}}
     Random Seed=?
    )
    Lifetime=0.09
    SkeletalMesh=$System.Skeletal Mesh
   )
  ]
  particleUpdateScript [
   &FN_UpdateAnimTrail(Locked Particle RibbonWidth=1, SkeletalMesh=$System.Skeletal Mesh) {module=FN_UpdateAnimTrail001}
   &ScaleColor(
    Scale Alpha=&FloatFromCurve(
     FloatCurve=DataInterface() {propertyValues {Curve {keys [
         {arriveTangent=-2.003, interpMode=RCIM_Cubic, leaveTangent=-2.003, tangentMode=RCTM_User, time=0.5, value=1}
         {interpMode=RCIM_Cubic, tangentMode=RCTM_User, time=1, value=0}
        ]}}}
     Scale Curve=&FN_Select_Float_from_Time_Of_Day(Morning=0.32, Day=0.42, Evening=0.35, Night=0.28)
    )
    Scale RGB=&MakeVectorFromLinearColorRGB(Color=&ColorFromCurve(ColorCurve=DataInterface() {propertyValues {
        AlphaCurve {keys []}
        BlueCurve {keys [<at 0 1.2 RCIM_Cubic>, {arriveTangent=-0.5097, interpMode=RCIM_Cubic, leaveTangent=-0.5097, time=0.077, value=1}, <at 0.9809 0.7>]}
        GreenCurve {keys [<at 0 1.2 RCIM_Cubic>, {arriveTangent=-0.5097, interpMode=RCIM_Cubic, leaveTangent=-0.5097, time=0.077, value=1}, <at 0.9809 0.7>]}
        RedCurve {keys [<at 0 1.2 RCIM_Cubic>, {arriveTangent=-0.5097, interpMode=RCIM_Cubic, leaveTangent=-0.5097, time=0.077, value=1}, <at 0.9809 0.7>]}
       }}))
    ScaleA=true
    ScaleRGB=true
   )
   &SolveForcesAndVelocity(Clamp Velocity=false, Limit Acceleration=false)
   &ScaleColor(
    Scale Alpha=&FloatFromCurve(
     FloatCurve=DataInterface() {propertyValues {Curve {keys [
         {interpMode=RCIM_Cubic, tangentMode=RCTM_User, time=0.15, value=1}
         {arriveTangent=-0.2118, interpMode=RCIM_Cubic, leaveTangent=-0.2118, tangentMode=RCTM_User, time=0.25, value=0}
        ]}}}
     CurveIndex=$System.Age
    )
    ScaleA=true
    ScaleRGB=false
   ) {module=ScaleColor001}
  ]
  rendererClasses [&NiagaraRibbonRendererProperties]
  renderers [{rendererClass=&NiagaraRibbonRendererProperties}]
 }
 BladeTrailBlack {
  emitterUpdateScript [
   &EmitterState()
   &FN_SpawnAnimTrail(Spawn Interval=0)
  ]
  particleSpawnScript [
   &FN_InitializeAnimTrailBoneIndecies(First Bone or Socket Index=1, Second Bone or Socket Index=2)
   &FN_InitializeAnimTrail(Color=<rgba 0.004 0.004 0.004 1>, Lifetime=0.1, SkeletalMesh=$System.Skeletal Mesh)
  ]
  particleUpdateScript [
   &FN_UpdateAnimTrail(Locked Particle RibbonWidth=1, SkeletalMesh=$System.Skeletal Mesh) {module=FN_UpdateAnimTrail001}
   &ScaleColor(
    Scale Alpha=&FloatFromCurve(
     FloatCurve=DataInterface() {propertyValues {Curve {keys [
         {arriveTangent=-2.003, interpMode=RCIM_Cubic, leaveTangent=-2.003, tangentMode=RCTM_User, time=0.5, value=1}
         {interpMode=RCIM_Cubic, tangentMode=RCTM_User, time=1, value=0}
        ]}}}
     Scale Curve=&FN_Select_Float_from_Time_Of_Day(Morning=0.8, Day=0.8, Evening=0.75, Night=0.85)
    )
    ScaleA=true
    ScaleRGB=false
   )
   &SolveForcesAndVelocity(Clamp Velocity=false, Limit Acceleration=false)
   &ScaleColor(
    Scale Alpha=&FloatFromCurve(
     FloatCurve=DataInterface() {propertyValues {Curve {keys [
         {interpMode=RCIM_Cubic, tangentMode=RCTM_User, time=0.15, value=1}
         {arriveTangent=-0.2118, interpMode=RCIM_Cubic, leaveTangent=-0.2118, tangentMode=RCTM_User, time=0.25, value=0}
        ]}}}
     CurveIndex=$System.Age
    )
    ScaleA=true
    ScaleRGB=false
   ) {module=ScaleColor001}
  ]
  rendererClasses [&NiagaraRibbonRendererProperties]
  renderers [{rendererClass=&NiagaraRibbonRendererProperties}]
 }
 BladeTrailMain {
  emitterUpdateScript [
   &EmitterState()
   &FN_SpawnAnimTrail(Spawn Interval=0)
  ]
  particleSpawnScript [
   &FN_InitializeAnimTrailBoneIndecies(First Bone or Socket Index=1, Second Bone or Socket Index=0)
   &FN_InitializeAnimTrail(
    Color=&SelectLinearColorFromArray(
     Array Sampling Mode=::NewEnumerator1
     Direct Array Index=&MakeIntFromFloat(Float=$User.HitCount)
     Color Selection Array=DataInterface() {propertyValues {ColorData [
        <rgba 0.3 0.1893 0.054 1>
        <rgba 0.3 0.1893 0.054 1>
        <rgba 0.3 0.1893 0.054 1>
        <rgba 0.3 0.1893 0.054 1>
        <rgba 0.3 0.1893 0.054 1>
        <rgba 0.9 0.018 0.0327 1>
       ]}}
     Random Seed=?
    )
    Lifetime=0.16
    SkeletalMesh=$System.Skeletal Mesh
   )
  ]
  particleUpdateScript [
   &FN_UpdateAnimTrail(Locked Particle RibbonWidth=1, SkeletalMesh=$System.Skeletal Mesh) {module=FN_UpdateAnimTrail001}
   &ScaleColor(
    Scale Alpha=&FloatFromCurve(
     FloatCurve=DataInterface() {propertyValues {Curve {keys [
         {arriveTangent=-2.003, interpMode=RCIM_Cubic, leaveTangent=-2.003, tangentMode=RCTM_User, time=0.5, value=1}
         {interpMode=RCIM_Cubic, tangentMode=RCTM_User, time=1, value=0}
        ]}}}
     Scale Curve=&FN_Select_Float_from_Time_Of_Day(Morning=0.55, Day=0.67, Evening=0.5, Night=0.45)
    )
    Scale RGB=&MakeVectorFromLinearColorRGB(Color=&ColorFromCurve(ColorCurve=DataInterface() {propertyValues {
        AlphaCurve {keys []}
        BlueCurve {keys [<at 0 1.2 RCIM_Cubic>, {arriveTangent=-0.5097, interpMode=RCIM_Cubic, leaveTangent=-0.5097, time=0.077, value=1}, <at 0.9809 0.7>]}
        GreenCurve {keys [<at 0 1.2 RCIM_Cubic>, {arriveTangent=-0.5097, interpMode=RCIM_Cubic, leaveTangent=-0.5097, time=0.077, value=1}, <at 0.9809 0.7>]}
        RedCurve {keys [<at 0 1.2 RCIM_Cubic>, {arriveTangent=-0.5097, interpMode=RCIM_Cubic, leaveTangent=-0.5097, time=0.077, value=1}, <at 0.9809 0.7>]}
       }}))
    ScaleA=true
    ScaleRGB=true
   )
   &SolveForcesAndVelocity(Clamp Velocity=false, Limit Acceleration=false)
   &ScaleColor(
    Scale Alpha=&FloatFromCurve(
     FloatCurve=DataInterface() {propertyValues {Curve {keys [
         {interpMode=RCIM_Cubic, tangentMode=RCTM_User, time=0.15, value=1}
         {arriveTangent=-0.2118, interpMode=RCIM_Cubic, leaveTangent=-0.2118, tangentMode=RCTM_User, time=0.25, value=0}
        ]}}}
     CurveIndex=$System.Age
    )
    ScaleA=true
    ScaleRGB=false
   ) {module=ScaleColor001}
  ]
  rendererClasses [&NiagaraRibbonRendererProperties]
  renderers [{rendererClass=&NiagaraRibbonRendererProperties}]
 }
}
summary {
 systemName=NS_Pickaxe_ProdigyHaughty_Trails
 emitters [
  {emitterName=BladeTrailMain, rendererClasses [&NiagaraRibbonRendererProperties]}
  {emitterName=BladeTrailBlack, rendererClasses [&NiagaraRibbonRendererProperties]}
  {emitterName=BladeSmall, rendererClasses [&NiagaraRibbonRendererProperties]}
 ]
 userVariables [
  User.bisSwinging {type=NiagaraBool}
  User.Color A {type=LinearColor, description="", defaultValue {struct=LinearColor, value=<rgba 1 1 1 1>}}
  User.HitCount {type=NiagaraFloat, defaultValue=0, description=""}
  User.RibbonIDOffset {type=NiagaraInt32}
  User.SkeletalMesh {
   type=NiagaraDataInterfaceSkeletalMesh
   defaultValue {
    struct=NiagaraExt_VariableValue_DataInterface
    value {dataInterface=&self:NiagaraDataInterfaceSkeletalMesh_1}
   }
  }
 ]
}
