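# Material function for the ProdigyHaughty pickaxe that builds progressive glow masks for its gems and blades from one EffectAmount input.
# UV square masks from a reused ProdigySage function split the red mask channel into gem regions, faded in over stages spanning 0 to 3.
# From 3 to 5, shifting local-X gradients reveal the green, then blue blade channels.
# It outputs gem, blade and combined masks.
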MaterialFunction /BRCosmetics/Weapons/FORT_Melee/Pickaxe_ProdigyHaughty/Materials/MaterialFunctions/MF_Pickaxe_ProdigyHaughty_GlowMasks
graph {nodes [
  n1 = TextureSample() {samplerType=SAMPLERTYPE_Masks, texture=/BRCosmetics/Weapons/FORT_Melee/Pickaxe_ProdigyHaughty/Textures/T_ProdigyHaughty_Pickaxe_Masks}
  n2 = MaterialFunctionCall!(
   Texture Coordinate=TextureCoordinate()
   X Pos Right=Constant() {r=0.4}
   X Pos Left=Constant() {r=0.02}
   Y Pos Up=Constant() {r=0.02}
   Y Pos Down=Constant() {r=0.4}
  ) {materialFunction=/BRCosmetics/Weapons/FORT_Melee/Pickaxe_ProdigySage/Materials/MaterialFuncions/MF_Pickaxe_ProdigySage_SquareMaskGenerator}
  n3 = FunctionInput() {bUsePreviewValueAsDefault=true, inputName=EffectAmount, inputType=FunctionInput_Scalar}
  n4 = Multiply(A=@n2.SquareMask, B=Saturate(None=InverseLinearInterpolate(Value=@n3)))
  n5 = MaterialFunctionCall(
   Texture Coordinate=TextureCoordinate()
   X Pos Right=Constant() {r=0.6}
   X Pos Left=Constant() {r=0.1}
   Y Pos Up=Constant() {r=0.7}
   Y Pos Down=Constant() {r=0.9}
  ) {materialFunction=&MF_Pickaxe_ProdigySage_SquareMaskGenerator}
  n6 = MaterialFunctionCall(
   Texture Coordinate=TextureCoordinate()
   X Pos Right=Constant() {r=0.95}
   X Pos Left=Constant() {r=0.45}
   Y Pos Up=Constant() {r=0.15}
   Y Pos Down=Constant() {r=0.3}
  ) {materialFunction=&MF_Pickaxe_ProdigySage_SquareMaskGenerator}
  n7 = Multiply(A=Saturate(None=InverseLinearInterpolate(Value=@n3) {constA=1, constB=2}), B=Add(A=@n5.SquareMask, B=@n6.SquareMask))
  n8 = Saturate(None=InverseLinearInterpolate(Value=@n3) {constA=2, constB=3})
  n9 = MaterialFunctionCall(
   Texture Coordinate=TextureCoordinate()
   X Pos Right=Constant() {r=0.95}
   X Pos Left=Constant() {r=0.45}
   Y Pos Up=Constant() {r=0.3}
   Y Pos Down=Constant() {r=0.6}
  ) {materialFunction=&MF_Pickaxe_ProdigySage_SquareMaskGenerator}
  n10 = Multiply(A=@n1.R, B=Add(A=Add(A=@n4, B=@n7), B=Multiply(A=@n8, B=@n9.SquareMask)))
  n11 = Constant() {r=0.075}
  n12 = Multiply(
   A=ComponentMask(None=MaterialFunctionCall() {materialFunction=&LocalPosition}.Local Position) {r=true}
   B=@n11
  ) {constB=0.5}
  n13 = InverseLinearInterpolate(Value=@n3) {constA=3, constB=4}
  n14 = Multiply(A=@n11, B=LinearInterpolate(Alpha=Saturate(None=@n13)) {constA=4, constAlpha=1, constB=36})
  n15 = Multiply(A=@n1.G, B=Saturate(None=Add(A=@n12, B=@n14) {constB=10}))
  n16 = Constant() {r=0.075}
  n17 = Multiply(
   A=ComponentMask(None=MaterialFunctionCall() {materialFunction=&LocalPosition}.Local Position) {r=true}
   B=@n16
  ) {constB=0.5}
  n18 = InverseLinearInterpolate(Value=@n3) {constA=4, constB=5}
  n19 = Multiply(A=@n16, B=LinearInterpolate(Alpha=Saturate(None=@n18)) {constA=11, constAlpha=1, constB=72})
  n20 = Multiply(A=@n1.B, B=Saturate(None=Add(A=@n17, B=@n19) {constB=10}))
  FunctionOutput(None=Add(A=Add(A=@n10, B=@n15), B=@n20)) {outputName=FullMask}
  FunctionOutput(None=@n10) {outputName=GemsMask, sortPriority=1}
  FunctionOutput(None=Add(A=@n15, B=@n20)) {outputName=BladesMask, sortPriority=2}
 ]}
