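# Parent material for the ProdigyHaughty pickaxe, adding a reactive red glow, grouped under WeakSpotHit, to the shared 2023 character shading functions.
# ReactivityStatus switches effect amount and glow multiplier between parameter-driven, fixed and zero values, with the amount feeding a custom glow-mask function.
# The full mask tints base color red.
# Gem and scaled blade masks drive a sine-pulsed, colored glow that passes through eye adaptation and replaces the emissive output.
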Material /BRCosmetics/Weapons/FORT_Melee/Pickaxe_ProdigyHaughty/Materials/M_ProdigyHaughty_PARENT
graph {
 properties {
  bUseMaterialAttributes=true
  bUsedWithInstancedStaticMeshes=true
  bUsedWithMeshDeformer=true
  bUsedWithNanite=true
  bUsedWithSkeletalMesh=true
  pixelDepthOffsetMode=PDOM_Legacy
  usedShadingModels=Default Lit
 }
 nodes [
  n1 = MaterialFunctionCall() {materialFunction=&MF_CharacterParent_2023_PreFX}
  n2 = GetMaterialAttributes(None=MaterialFunctionCall(Material=@n1.Result, M=@n1.Result, Pre Skinned Local Position=@n1.Result, Distance Blend=@n1.Result, Fresnel=@n1.Result) {materialFunction=&MF_CharacterParent_2023_PostFX}.Result) {attributeGetTypes [#1]}
  n3 = ScalarParameter() {group=WeakSpotHit, parameterName=ReactivityStatus, sliderMax=2, sortPriority=20}
  n4 = MaterialFunctionCall!(EffectAmount=If(A=@n3, "A > B"=Constant(), "A == B"=Constant() {r=5}, "A < B"=ScalarParameter() {group=WeakSpotHit, parameterName=EffectAmount, sliderMax=5}) {constB=1}) {materialFunction=/BRCosmetics/Weapons/FORT_Melee/Pickaxe_ProdigyHaughty/Materials/MaterialFunctions/MF_Pickaxe_ProdigyHaughty_GlowMasks}
  n5 = VectorParameter() {group=WeakSpotHit, parameterName=GlowColor, sortPriority=31, defaultValue=<rgba 0.4621 0.0262 0.0273 1>}
  n6 = Multiply(A=@n4.BladesMask, B=ScalarParameter() {group=WeakSpotHit, parameterName=BladesGlowMultiplier, sliderMax=1, sortPriority=34}) {constB=0.1}
  n7 = Multiply(
   A=Multiply(A=@n5.RGB, B=Add(A=@n4.GemsMask, B=@n6))
   B=ScalarParameter() {defaultValue=1, group=WeakSpotHit, parameterName=OveralGlowIntensity, sortPriority=33}
  )
  n8 = LinearInterpolate(Alpha=If(
    A=@n3
    "A > B"=Constant()
    "A == B"=Constant() {r=0.5}
    "A < B"=ScalarParameter() {group=WeakSpotHit, parameterName=OveralGlowMultiplier, sliderMax=2, sortPriority=34}
   ) {constB=1}) {constB=11}
  n9 = Multiply(A=Time()) {constB=0.4}
  n10 = Add(A=Multiply(A=Sine(None=@n9)) {constB=0.5}) {constB=0.5}
  n11 = MaterialFunctionCall!(
   Emissive=Multiply(A=Multiply(A=@n7, B=@n8), B=LinearInterpolate(Alpha=@n10) {constA=0.5})
   Alpha=Constant() {r=1}
   Material=@n2.MaterialAttributes
  ) {materialFunction=/Game/Packages/Fortress_SharedMaterials/Base_Material_Functions/BaseCharacterShaderFunctions/MF_CosmeticEmissiveEyeAdaptation}
  SetMaterialAttributes!(
   MaterialAttributes=@n2.MaterialAttributes
   Base Color=LinearInterpolate(A=@n2.Base Color, B=Constant3Vector() {constant=<rgba 0.4688 0.0221 0.023 1>}, Alpha=@n4.FullMask)
   Emissive Color=@n11.Emissive
  ) {attributeSetTypes [#1, #2]}
 ]
}
